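# Conversation database for GoldenPleats, an NPC tagged as a Medic Specialist, supplying the greeting and routing to shared conversations.
# Its entry point leads to one Speech node whose contextual lines react to a first-ever conversation, three specific player outfits, or per-match conversation count.
# Priorities for the first-meeting and outfit lines come from named curve rows.
# The speech then branches through link nodes to NPC Library quests, BasicServices and the Ch5S1.D3.Q8 conversation.
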ConversationDatabase /NPCLibrary/NPCs/GoldenPleats/NPCLibrary_Conversation_GoldenPleats
entry_points [{tag=Conversation.Entry.GoldenPleats, destinations [#1]}]
nodes [
 ConversationEntryPointNode_7 {
  id=#1
  class=ConversationEntryPointNode
  outputs [#2]
  properties {entryTag {tagName=Conversation.Entry.GoldenPleats}}
 }
 Speech_C_4 {
  id=#2
  class=Speech_C
  outputs [#3, #4, #5]
  properties {
   nodeName=""
   generalConfig {
    defaultMessage="You're back? But, no one ever comes back..."
    contextualMessages [
     {
      message=Look out for yourself out there. No one else will.
      contextRequirements [{requirement {class=FortControllerRequirement_IsFirstEverConversationWithNPC}}]
      priority {curve {rowName=Tandem.Speech.Priority.FirstEverInteraction}}
     }
     {
      message="So, you found another crew. Wonder when you'll leave THEM behind."
      contextRequirements [{requirement {class=FortControllerRequirement_HasCID, properties {allowedCIds [/BRCosmetics/Athena/Items/Cosmetics/Characters/Character_BlazerVeil]}}}]
      priority {curve {rowName=Tandem.Speech.Priority.OutfitReaction}}
     }
     {
      message="I told you, Jones. I'll help you this once, but I work alone."
      contextRequirements [{requirement {
         class=FortControllerRequirement_HasCID
         properties {allowedCIds [&Character_UndergroundRebel_Fashion]}
        }}]
      priority {curve {rowName=Tandem.Speech.Priority.OutfitReaction}}
     }
     {
      message=You used to be a nomad. Why'd you give it up?
      contextRequirements [{requirement {class=FortControllerRequirement_HasCID, properties {allowedCIds [/BRCosmetics/Athena/Items/Cosmetics/Characters/Character_ZirconSweep]}}}]
      priority {curve {rowName=Tandem.Speech.Priority.OutfitReaction}}
     }
     {
      message="You're back? But, no one ever comes back..."
      contextRequirements [{requirement {
         class=FortControllerRequirement_CheckNPCMatchConversationCount
         properties {nPCUniqueIdTag {tagName=AISpawnerData.Type.Specialist.Medic.GoldenPleats}}
        }}]
      priority {curve {rowName=None}}
     }
     {
      message=The Society is falling apart. I guess change really is possible.
      contextRequirements [{requirement {
         class=FortControllerRequirement_CheckNPCMatchConversationCount
         properties {nPCUniqueIdTag {tagName=AISpawnerData.Type.Specialist.Medic.GoldenPleats}}
        }}]
      priority {curve {rowName=None}}
     }
    ]
   }
  }
 }
 ConversationLinkNode_2 {id=#5, class=ConversationLinkNode, properties {remoteEntryTag {tagName=Conversation.Entry.Ch5S1.D3.Q8}}}
 ConversationLinkNode_0 {id=#4, class=ConversationLinkNode, properties {remoteEntryTag {tagName=Conversation.Entry.BasicServices}}}
 ConversationLinkNode_1 {id=#3, class=ConversationLinkNode, properties {remoteEntryTag {tagName=Conversation.Entry.NPCLibrary.Quests}}}
]
